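# Sound cue for the basketball toy's jump shot in Athena.
# A single wave player feeds a modulator that adds very slight random pitch variation on each play.
# The cue raises overall pitch slightly, lowers volume, and routes through the emote music sound class with Athena emote attenuation.
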SoundCue /Game/Athena/Sounds/Toys/Basketball/Toys_BasketBall_Jump_Shoot_Cue
root=Modulator_0
nodes [
 {id=Modulator_0, children [WavePlayer_1], properties {pitchMax=1.01, pitchMin=0.99}}
 {id=WavePlayer_1, properties {soundWaveAssetPtr=&/Basketball_JumP_Shoot.Basketball_Jump_Shoot}}
]
properties {
 attenuationSettings=&Fort_Emotes_Athena_Att
 maxDistance=1600
 pitchMultiplier=1.1
 soundClassObject=&Fort_Music_Emote
 volumeMultiplier=0.8
}
