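# Sound cue for throwing the basketball toy in Athena.
# A Random node picks between two throw wave variations with equal weight.
# The cue slightly raises pitch and lowers volume, uses the shared Athena emote attenuation, and is assigned to the emote music sound class.
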SoundCue /Game/Athena/Sounds/Toys/Basketball/Toys_BasketBall_Throw_Cue
root=Random_1
nodes [
 {id=Random_1, children [WavePlayer_2,WavePlayer_3], properties {weights [1,1]}}
 {id=WavePlayer_2, properties {soundWaveAssetPtr=&/Toys_Basketball_Throw_v2}}
 {id=WavePlayer_3, properties {soundWaveAssetPtr=&/Toys_Basketball_Throw_v1}}
]
properties {
 attenuationSettings=&Fort_Emotes_Athena_Att
 maxDistance=1600
 pitchMultiplier=1.1
 soundClassObject=&Fort_Music_Emote
 volumeMultiplier=0.8
}
